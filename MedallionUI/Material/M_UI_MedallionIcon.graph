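# Translucent UI material for a medallion icon, with a colored outline and an optional light wipe.
# The outline comes from the alpha difference between the icon texture and a center-scaled sample of it, tinted by OutlineColor.
# The wipe draws anti-aliased SDF line segments in rotated UVs shifted by PanAmount, with an optional offset second line.
# ShowLightWipe blends the tinted wipe over the outlined icon.
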Material /MedallionUI/Material/M_UI_MedallionIcon
graph {
 properties {blendMode=BLEND_Translucent, materialDomain=MD_UI, pixelDepthOffsetMode=PDOM_Legacy, usedShadingModels=Default Lit}
 nodes [
  n1 = MaterialFunctionCall() {materialFunction=&GetUserInterfaceUV}
  n2 = TextureSampleParameter2D(UVs=@n1.Normalized UV) {parameterName=MedallionIconTexture, texture=/Game/UI/Foundation/Textures/Icons/HUD/T-Icon-Alert1-32}
  n3 = TextureSampleParameter2D(UVs=MaterialFunctionCall(UVs=@n1.Normalized UV, Texture Scale=ScalarParameter() {defaultValue=1.064}) {materialFunction=&ScaleUVsByCenter}.UVs) {parameterName=MedallionIconTexture, texture=&T-Icon-Alert1-32}
  LinearInterpolate(A=@n2.A, B=@n3.A) {constAlpha=0}
  n4 = TextureSampleParameter2D(UVs=@n1.Normalized UV) {parameterName=MedallionIconTexture, texture=&T-Icon-Alert1-32}
  n5 = Add(
   A=Subtract(A=@n4.RGB, B=OneMinus(None=@n4.A))
   B=Multiply(A=Subtract(A=@n3.A, B=@n2.A), B=VectorParameter() {parameterName=OutlineColor, defaultValue=<rgba 0.1458 0.0483 0.1338 1>}.RGB)
  )
  n6 = Add(A=@n5, B=NamedRerouteUsage())
  LinearInterpolate(A=@n5, B=LinearInterpolate(A=@n5, B=@n6, Alpha=@n6) {constAlpha=1}, Alpha=ScalarParameter() {parameterName=ShowLightWipe, sliderMax=1})
  n7 = ScalarParameter() {defaultValue=-0.25, parameterName=Line2Offset}
  n8 = MaterialFunctionCall!(
   UVs=Add(A=NamedRerouteUsage(), B=NamedRerouteUsage()) {constA=0.5}
   PointA=Add(A=@n7, B=NamedRerouteUsage())
   PointB=Add(A=@n7, B=NamedRerouteUsage())
  ) {materialFunction=/FortUILibrary/MaterialFunctions/Shapes/SDFs/MF_SDF_LineSegment}
  n9 = ScalarParameter() {defaultValue=0.08, parameterName=LineThickness2}
  n10 = ScalarParameter() {defaultValue=0.1, parameterName=LineFeathering2}
  NamedRerouteDeclaration(None=Multiply(
    A=MaterialFunctionCall(DistanceField=@n8.SDF, Threshold=@n9, Smoothing=@n10) {materialFunction=&MF_UI_AntiAliasing}.Inner
    B=VectorParameter() {parameterName=LightWipeColor, defaultValue=<rgba 1 0.9522 0.7223 1>}.RGB
   )) {name=SecondLightWipe, nodeColor=<rgba 0 0 0 1>}
  n11 = ScalarParameter() {parameterName=PanAmount}
  n12 = TextureCoordinate()
  n13 = ScalarParameter() {parameterName=LightwipeAngle, sliderMax=1}
  n14 = Add(
   A=NamedRerouteDeclaration(None=ConstantBiasScale(None=@n11) {bias=-0.5, scale=-1}) {name=Panning, nodeColor=<rgba 0 0 0 1>}
   B=NamedRerouteDeclaration(None=MaterialFunctionCall(UVs=@n12, "Rotation Angle (0-1)"=@n13) {materialFunction=&CustomRotator}.Rotated Values) {name=Rotated UVs, nodeColor=<rgba 0 0 0 1>}
  ) {constA=0.5}
  n15 = VectorParameter() {parameterName=LinePoints, channelNames {a=PointBY, b=PointBX, g=PointAY, r=PointAX}, defaultValue=<rgba 1 0 0 1>}
  n16 = NamedRerouteDeclaration(None=ConstantBiasScale(None=AppendVector(A=@n15.PointAX, B=@n15.PointAX)) {bias=-0.5, scale=2}) {name=LinePointA, nodeColor=<rgba 0 0 0 1>}
  n17 = NamedRerouteDeclaration(None=ConstantBiasScale(None=AppendVector(A=@n15.PointBX, B=@n15.PointBX)) {bias=-0.5, scale=2}) {name=LinePointB, nodeColor=<rgba 0 0 0 1>}
  n18 = Multiply(
   A=MaterialFunctionCall(
    DistanceField=MaterialFunctionCall(UVs=@n14, PointA=@n16, PointB=@n17) {materialFunction=&MF_SDF_LineSegment}.SDF
    Threshold=ScalarParameter() {defaultValue=0.1, parameterName=LineThickness1}
    Smoothing=ScalarParameter() {defaultValue=0.1, parameterName=LineFeathering1}
   ) {materialFunction=&MF_UI_AntiAliasing}.Inner
   B=VectorParameter() {parameterName=LightWipeColor, defaultValue=<rgba 1 0.9522 0.7223 1>}.RGB
  )
  n19 = NamedRerouteUsage()
  NamedRerouteDeclaration(None=LinearInterpolate(A=@n18, B=Add(A=@n18, B=@n19), Alpha=ScalarParameter() {parameterName=ShowSecondLightWipe}) {constAlpha=1}) {name=LightWipeColor, nodeColor=<rgba 0 0 0 1>}
 ]
}
